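# Additive UI material that draws animated, flickering lightning, named for the ability boost in the frontend Hero Loadout.
# A four-column FlipBook cycles the lightning sheet, with a static switch for stepped, time-driven UV rotation built from threshold comparisons.
# The sheet's R, G and B channels are weighted and gated by staggered time windows, then summed into a mask.
# That mask blends four color parameters, while a repeating pulse modulates intensity and color.
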Material /Game/UI/Frontend/HeroLoadout/Art/M_UI_AbilityBoostLightning
graph {
 properties {bUsedWithStaticMesh=false, blendMode=BLEND_Additive, materialDomain=MD_UI, pixelDepthOffsetMode=PDOM_Legacy, usedShadingModels=Default Lit}
 nodes [
  n1 = Frac(None=Multiply(A=Time(), B=ScalarParameter() {defaultValue=2, parameterName=FlipSpeed}))
  n2 = Constant() {r=1}
  n3 = TextureCoordinate()
  n4 = Fmod(A=Multiply(A=Time(), B=ScalarParameter() {defaultValue=0.0625, parameterName=Speed}) {constB=0.0625}, B=Constant() {r=1})
  n5 = Constant() {r=0.75}
  n6 = Constant()
  n7 = If(A=@n4, B=@n5, "A > B"=@n5, "A == B"=@n6, "A < B"=@n6)
  n8 = Constant() {r=0.5}
  n9 = Constant()
  n10 = If(A=@n4, B=@n8, "A > B"=@n8, "A == B"=@n9, "A < B"=@n9)
  n11 = Constant() {r=0.25}
  n12 = Constant()
  n13 = If(A=@n4, B=@n11, "A > B"=@n11, "A == B"=@n12, "A < B"=@n12)
  n14 = Add(A=@n10, B=@n13)
  n15 = Clamp(None=Ceil(None=Add(A=@n7, B=@n14)))
  n16 = LinearInterpolate(B=@n7, Alpha=@n15)
  n17 = Add(A=LinearInterpolate(B=@n10, Alpha=@n15), B=LinearInterpolate(B=@n13, Alpha=@n15))
  n18 = Rotator(Coordinate=@n3, Time=Clamp(None=Add(A=@n16, B=@n17))) {speed=6.2832}
  n19 = TextureSample(UVs=MaterialFunctionCall(
    "Animation  Phase (0-1)"=Multiply(A=@n1, B=@n2)
    Number of Rows=Constant() {r=1}
    Number of Columns=Constant() {r=4}
    UVs=StaticSwitchParameter(True=@n18, False=@n3) {defaultValue=true, parameterName=Rotate}
   ) {materialFunction=&FlipBook}.UVs) {texture=&/T_UI_LightningSheet}
  n20 = Multiply(A=@n19.R)
  n21 = Multiply(A=Time(), B=ScalarParameter() {defaultValue=0.25, parameterName=TimeOffset1})
  n22 = ScalarParameter() {defaultValue=1, parameterName=fmod}
  n23 = Clamp(None=LinearInterpolate(A=Constant() {r=2}, B=Constant() {r=-16}, Alpha=Fmod(A=@n21, B=@n22)))
  n24 = Constant()
  n25 = Constant() {r=1}
  n26 = Multiply(A=DotProduct(A=Clamp(None=@n20), B=If(A=@n23, B=@n24, "A > B"=@n25, "A == B"=@n24, "A < B"=@n24))) {constB=0.5}
  n27 = Add(A=Time(), B=ScalarParameter() {defaultValue=-1, parameterName=TimeOffset2})
  n28 = Constant() {r=0.25}
  n29 = LinearInterpolate(A=Constant() {r=2}, B=Constant() {r=-16}, Alpha=Fmod(A=Multiply(A=@n27, B=@n28), B=@n22))
  n30 = Constant()
  n31 = DotProduct(A=Clamp(None=Multiply(A=@n19.G)), B=If(A=Clamp(None=@n29), B=@n30, "A > B"=Constant() {r=1}, "A == B"=@n30, "A < B"=@n30))
  n32 = Add(A=Time(), B=ScalarParameter() {defaultValue=-1.5, parameterName=TimeOffset3})
  n33 = Constant() {r=0.25}
  n34 = LinearInterpolate(A=Constant() {r=2}, B=Constant() {r=-16}, Alpha=Fmod(A=Multiply(A=@n32, B=@n33), B=@n22))
  n35 = Constant()
  n36 = DotProduct(A=Clamp(None=Multiply(A=@n19.B)), B=If(A=Clamp(None=@n34), B=@n35, "A > B"=Constant() {r=1}, "A == B"=@n35, "A < B"=@n35))
  n37 = Clamp(None=Add(A=Add(A=@n26, B=@n31), B=Multiply(A=@n36) {constB=0.75}))
  n38 = Multiply(A=Time(), B=Constant() {r=2})
  n39 = Constant() {r=1}
  n40 = Clamp(None=LinearInterpolate(A=Constant() {r=1}, B=Constant() {r=0.5}, Alpha=Fmod(A=@n38, B=@n39)))
  Clamp(None=LinearInterpolate(B=@n37, Alpha=@n40))
  n41 = VectorParameter() {parameterName=Color2, defaultValue=<rgba 0 0 1 1>}
  n42 = VectorParameter() {parameterName=Color1, defaultValue=<rgba 0 0.5 1 1>}
  n43 = Clamp(None=Power(Base=@n40) {constExponent=16})
  n44 = VectorParameter() {parameterName=Color4, defaultValue=<rgba 0 1 1 1>}
  n45 = VectorParameter() {parameterName=Color3, defaultValue=<rgba 1 1 1 1>}
  Clamp(None=LinearInterpolate(A=LinearInterpolate(A=@n41.RGB, B=@n42.RGB, Alpha=@n43), B=LinearInterpolate(A=@n44.RGB, B=@n45.RGB, Alpha=@n43), Alpha=@n37))
  Constant() {r=1}
  Constant() {r=1}
  Constant() {r=1}
 ]
}
