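# UI material for a Hero Loadout warning icon: a dark warning glyph on a round red-to-amber badge.
# An exponential radial gradient, tuned by Radius and Density parameters, blends red at the rim to amber inside.
# The warning texture's alpha, rescaled about center, nudged vertically and masked by a tighter radial gradient, is subtracted from that color.
# The outer gradient, multiplied and clamped, forms the translucent disc shape.
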Material /Game/UI/Frontend/HeroLoadout/Art/M_UI_LoadoutWarningIcon
graph {
 properties {
  bUsedWithStaticMesh=false
  blendMode=BLEND_Translucent
  materialDomain=MD_UI
  pixelDepthOffsetMode=PDOM_Legacy
  usedShadingModels=Default Lit
 }
 nodes [
  n1 = Constant3Vector() {constant=<rgba 1 0.2667 0 0>}
  n2 = Constant3Vector() {constant=<rgba 0.7991 0.521 0 0>}
  n3 = TextureCoordinate()
  n4 = MaterialFunctionCall(
   UVs=@n3
   Radius=ScalarParameter() {defaultValue=0.5, parameterName="Radius (S)"}
   Density=ScalarParameter() {defaultValue=2.3, parameterName="Density (S)"}
  ) {materialFunction=&RadialGradientExponential}
  n5 = Clamp(None=Multiply(A=@n4.RadialGradientExponential) {constB=4})
  n6 = Constant() {r=0.65}
  n7 = Constant()
  n8 = Constant() {r=0.075}
  n9 = TextureSample(UVs=Add(
    A=MaterialFunctionCall(UVs=@n3, Texture Scale=@n6) {materialFunction=&ScaleUVsByCenter}.UVs
    B=AppendVector(A=@n7, B=@n8)
   )) {texture=/Game/UI/Foundation/Shell/Textures/T-Icon-Warning-32}
  n10 = Constant() {r=0.3475}
  n11 = Clamp(None=Multiply(A=MaterialFunctionCall(UVs=@n3, Radius=@n10) {materialFunction=&RadialGradientExponential}.RadialGradientExponential) {constB=12})
  Clamp(None=Subtract(A=LinearInterpolate(A=@n1, B=@n2, Alpha=@n5), B=LinearInterpolate(B=@n9.A, Alpha=@n11)))
  Clamp(None=Multiply(A=@n4.RadialGradientExponential) {constB=8})
 ]
}
