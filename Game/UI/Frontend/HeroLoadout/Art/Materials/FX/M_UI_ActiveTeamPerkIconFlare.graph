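# Translucent UI material for an animated glow flare, named for the active team perk icon in the Hero Loadout screen.
# Three radially warped, rotating samples of blurred noise multiply into a swirl, optionally maxed with two counter-rotating angular beam layers (ShowBeams).
# A sine-pulsed radial gradient masks the result, which is shaped by Mult and Pow.
# Color comes from either a two-color blend or a curve atlas gradient (UseColorCurve).
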Material /Game/UI/Frontend/HeroLoadout/Art/Materials/FX/M_UI_ActiveTeamPerkIconFlare
graph {
 properties {
  bUsedWithStaticMesh=false
  blendMode=BLEND_Translucent
  materialDomain=MD_UI
  numCustomizedUVs=2
  pixelDepthOffsetMode=PDOM_Legacy
  usedShadingModels=Default Lit
 }
 nodes [
  n1 = TextureCoordinate()
  n2 = OneMinus(None=MaterialFunctionCall(UVs=TextureCoordinate(), CenterPosition=Constant2Vector() {g=0.5, r=0.5}, Radius=Constant() {r=0.8}, Density=Constant() {r=1.666}) {materialFunction=&RadialGradientExponential}.RadialGradientExponential)
  n3 = Time()
  n4 = ScalarParameter() {defaultValue=1, parameterName=SpeedMultiplier}
  n5 = Multiply(A=@n3, B=Multiply(A=@n4) {constB=0.075})
  n6 = Multiply(A=Rotator(
    Coordinate=MaterialFunctionCall(UVs=@n1, Texture Scale=@n2) {materialFunction=&ScaleUVsByCenter}.UVs
    Time=@n5
   ) {speed=-0.075}) {constB=0.5}
  n7 = MaterialFunctionCall(UVs=TextureCoordinate(), CenterPosition=Constant2Vector() {g=0.5, r=0.5}, Radius=Constant() {r=1}, Density=Constant() {r=1}) {materialFunction=&RadialGradientExponential}
  n8 = Rotator(
   Coordinate=MaterialFunctionCall(UVs=TextureCoordinate(), Texture Scale=OneMinus(None=@n7.RadialGradientExponential)) {materialFunction=&ScaleUVsByCenter}.UVs
   Time=Multiply(A=@n3, B=Multiply(A=@n4) {constB=0.25})
  )
  n9 = MaterialFunctionCall(UVs=TextureCoordinate(), CenterPosition=Constant2Vector() {g=0.5, r=0.5}, Radius=Constant() {r=1}, Density=Constant() {r=1.333}) {materialFunction=&RadialGradientExponential}
  n10 = Rotator(
   Coordinate=MaterialFunctionCall(UVs=TextureCoordinate(), Texture Scale=OneMinus(None=@n9.RadialGradientExponential)) {materialFunction=&ScaleUVsByCenter}.UVs
   Time=Multiply(A=@n3, B=Multiply(A=@n4) {constB=0.05})
  ) {speed=0.05}
  n11 = DotProduct(
   A=TextureSampleParameter2D(UVs=Multiply(A=@n8) {constB=0.5}) {parameterName=Texture, samplerType=SAMPLERTYPE_Grayscale, texture=&LowResBlurredNoise}.G
   B=TextureSampleParameter2D(UVs=Multiply(A=@n10) {constB=0.5}) {parameterName=Texture, samplerType=SAMPLERTYPE_Grayscale, texture=&LowResBlurredNoise}.B
  )
  n12 = Multiply(A=DotProduct(
    A=TextureSampleParameter2D(UVs=@n6) {parameterName=Texture, samplerType=SAMPLERTYPE_Grayscale, texture=&LowResBlurredNoise}.R
    B=Multiply(A=@n11) {constB=2}
   )) {constB=3}
  n13 = Divide(A=Time(), B=ScalarParameter() {defaultValue=1, group=Beams, parameterName=BeamSpeed})
  n14 = VectorParameter() {group=Beams, parameterName=VectorOffset, defaultValue=<rgba 0.5 -1 0 1>}
  n15 = Add(A=Rotator(Coordinate=TextureCoordinate() {coordinateIndex=1}, Time=@n13) {speed=2}, B=Add(A=@n14.R, B=@n14.R))
  n16 = ScalarParameter() {defaultValue=20, group=Beams, parameterName=Repeat}
  n17 = TextureSampleParameter2D(UVs=Multiply(
    A=MaterialFunctionCall(Vector or UVs=@n15) {materialFunction=&VectorToRadialValue}.Vector Converted to Angle
    B=Floor(None=@n16)
   )) {group=Beams, parameterName=SecondaryTexture, texture=/Game/Global/Textures/Default/Masks_and_Gradients/T_Vert_Gradient_Sections}
  n18 = TextureCoordinate() {coordinateIndex=1}
  n19 = Multiply(A=Time(), B=ScalarParameter() {defaultValue=2, group=Beams, parameterName=Beam2Speed})
  n20 = VectorParameter() {group=Beams, parameterName=VectorOffset, defaultValue=<rgba 0.5 -1 0 1>}
  n21 = MaterialFunctionCall(Vector or UVs=Add(A=Rotator(Coordinate=@n18, Time=@n19) {speed=-1}, B=Add(A=@n20.R, B=@n20.R))) {materialFunction=&VectorToRadialValue}
  n22 = Floor(None=ScalarParameter() {defaultValue=20, group=Beams, parameterName=Repeat})
  n23 = Multiply(
   A=@n17.R
   B=TextureSampleParameter2D(UVs=Multiply(A=@n21.Vector Converted to Angle, B=@n22)) {parameterName=BeamTexture2, texture=/Game/Global/Textures/Default/Masks_and_Gradients/T_UI_ReticleSegmentGradient}.R
  )
  n24 = Multiply(A=Time()) {constB=0.25}
  n25 = MaterialFunctionCall(
   UVs=TextureCoordinate()
   CenterPosition=Constant2Vector() {g=0.5, r=0.5}
   Radius=LinearInterpolate(A=Constant() {r=0.4}, B=Constant() {r=0.45}, Alpha=Sine(None=@n24))
   Density=Constant() {r=1.666}
  ) {materialFunction=&RadialGradientExponential}
  n26 = LinearInterpolate(B=StaticSwitchParameter(True=Max(A=@n12, B=@n23), False=@n12) {group=Beams, parameterName=ShowBeams}, Alpha=@n25.RadialGradientExponential)
  n27 = ScalarParameter() {defaultValue=4, parameterName=Mult}
  Clamp(None=Power(Base=Multiply(A=@n26, B=@n27) {constB=4}, Exp=ScalarParameter() {defaultValue=1, parameterName=Pow}))
  Fmod(A=@n13, B=Constant() {r=1})
  Fmod(A=@n5, B=Constant() {r=1})
  Fmod(A=Divide(A=Time(), B=ScalarParameter() {defaultValue=1, parameterName=BeamSpeed}), B=Constant() {r=1})
  MaterialFunctionCall(UVs=TextureCoordinate(), Texture Scale=ScalarParameter() {defaultValue=1, parameterName=UV1Scale}) {materialFunction=&ScaleUVsByCenter}
  n28 = Add(A=Multiply(A=Power(Base=@n25.RadialGradientExponential) {constExponent=4}) {constB=0.25}, B=@n26)
  n29 = CurveAtlasRowParameter(CurveTime=@n28) {
   atlas=/Game/Athena/HUD/Materials/Victory/NewCurveLinearColorAtlas
   curve=/Game/Athena/Playlists/Goose/UI/Material/GooseVictoryGradient1
   defaultValue=9
   parameterName=CurveParams
  }
  StaticSwitchParameter(
   True=@n29
   False=LinearInterpolate(
    A=VectorParameter() {parameterName=Color2, defaultValue=<rgba 0.1 0 1 1>}.RGB
    B=VectorParameter() {parameterName=Color1, defaultValue=<rgba 0 1 1 1>}.RGB
    Alpha=@n28
   )
  ) {parameterName=UseColorCurve}
  TextureCoordinate()
 ]
}
