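# Translucent UI material in the hero loadout frontend art that draws an animated, two-color noise flare.
# A procedural UV mask fades toward all four edges, with scalar dot products serving as multiplies.
# Two panning samples of blurred noise are multiplied and sharpened by Hardness to break up that mask.
# The mask and noise blend Color2 toward Color1 and form a clamped alpha, with Intensity and FinalAlpha controls.
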Material /Game/UI/Frontend/HeroLoadout/Art/Materials/FX/M_UI_FlareNoise
graph {
 properties {
  bUsedWithStaticMesh=false
  blendMode=BLEND_Translucent
  materialDomain=MD_UI
  pixelDepthOffsetMode=PDOM_Legacy
  usedShadingModels=Default Lit
 }
 nodes [
  n1 = TextureCoordinate()
  n2 = ComponentMask(None=@n1) {g=true}
  n3 = OneMinus(None=@n2)
  n4 = Clamp(None=Multiply(A=DotProduct(A=@n2, B=@n3)) {constB=4})
  n5 = ComponentMask(None=@n1) {r=true}
  n6 = Multiply(A=OneMinus(None=@n5)) {constB=2}
  n7 = Multiply(A=@n5) {constB=2}
  n8 = DotProduct(A=DotProduct(A=@n4, B=Clamp(None=@n6)), B=DotProduct(A=@n4, B=Clamp(None=@n7)))
  n9 = Multiply(A=@n8) {constB=0.25}
  n10 = TextureCoordinate() {vTiling=0.75}
  n11 = AppendVector(A=ScalarParameter() {defaultValue=1, parameterName=UTile}, B=ScalarParameter() {defaultValue=1, parameterName=VTile})
  n12 = AppendVector(A=Constant() {r=0.05}, B=Constant() {r=-0.125})
  n13 = ScalarParameter() {defaultValue=1, parameterName=Pan1Speed}
  n14 = TextureSample(UVs=Panner(Coordinate=Multiply(A=@n10, B=@n11), Speed=Multiply(A=@n12, B=@n13)) {speedX=0.05, speedY=-0.125}) {samplerType=SAMPLERTYPE_Grayscale, texture=&LowResBlurredNoise}
  n15 = TextureCoordinate()
  n16 = AppendVector(A=ScalarParameter() {defaultValue=1, parameterName=UTile}, B=ScalarParameter() {defaultValue=1, parameterName=VTile})
  n17 = AppendVector(A=Constant() {r=0.1}, B=Constant() {r=0.05})
  n18 = ScalarParameter() {defaultValue=1, parameterName=Pan2Speed}
  n19 = TextureSample(UVs=Panner(Coordinate=Multiply(A=@n15, B=@n16), Speed=Multiply(A=@n17, B=@n18)) {speedX=0.1, speedY=0.05}) {samplerType=SAMPLERTYPE_Grayscale, texture=&LowResBlurredNoise}
  n20 = Power(Base=Multiply(A=DotProduct(A=@n14.R, B=@n19.R)) {constB=4}) {constExponent=1.5}
  n21 = ScalarParameter() {defaultValue=1, parameterName=Hardness}
  n22 = LinearInterpolate(B=Clamp(None=Multiply(A=@n8) {constB=1.75}), Alpha=Power(Base=Multiply(A=@n20, B=@n21), Exp=Multiply(A=@n21) {constB=0.75}))
  n23 = Multiply(B=@n22) {constA=1}
  n24 = ScalarParameter() {defaultValue=1, parameterName=Intensity}
  LinearInterpolate(
   A=VectorParameter() {parameterName=Color2, defaultValue=<rgba 0.0999 0 1 1>}.RGB
   B=VectorParameter() {parameterName=Color1, defaultValue=<rgba 0 0.5 1 1>}.RGB
   Alpha=Multiply(A=Add(A=@n9, B=@n23), B=@n24)
  )
  n25 = Add(A=Multiply(A=@n8) {constB=0.5}, B=Multiply(A=@n22) {constA=0.75, constB=0.5})
  Multiply(A=Clamp(None=Multiply(A=@n25, B=@n24)), B=ScalarParameter() {defaultValue=1, parameterName=FinalAlpha})
 ]
}
