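# Translucent UI material for a pulsing radial glow, apparently behind the check mark on the frontend hero loadout screen.
# A time-driven sine oscillates the radius and density of an exponential radial gradient.
# The gradient samples rapidly rotating UVs around a very slightly off-center point.
# The gradient blends two color parameters and is separately multiplied by an Intensity parameter.
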Material /Game/UI/Frontend/HeroLoadout/Art/Materials/FX/M_UI_LoadoutCheck_Glow
graph {
 properties {
  bUsedWithStaticMesh=false
  blendMode=BLEND_Translucent
  materialDomain=MD_UI
  pixelDepthOffsetMode=PDOM_Legacy
  usedShadingModels=Default Lit
 }
 nodes [
  n1 = MaterialFunctionCall(Speed=Constant() {r=0.25}) {materialFunction=&DebugTimeSine}
  n2 = MaterialFunctionCall(
   UVs=Rotator(Coordinate=TextureCoordinate(), Time=Time()) {speed=32}
   CenterPosition=Constant2Vector() {g=0.5, r=0.499}
   Radius=LinearInterpolate(A=Constant() {r=0.425}, B=Constant() {r=0.45}, Alpha=@n1.Result)
   Density=LinearInterpolate(A=Constant() {r=0.55}, B=Constant() {r=0.66}, Alpha=@n1.Result)
  ) {materialFunction=&RadialGradientExponential}
  LinearInterpolate(
   A=VectorParameter() {parameterName=Color1, defaultValue=<rgba 0 0 1 1>}.RGB
   B=VectorParameter() {parameterName=Color2, defaultValue=<rgba 0 1 1 1>}.RGB
   Alpha=@n2.RadialGradientExponential
  )
  Multiply(A=@n2.RadialGradientExponential, B=ScalarParameter() {defaultValue=1, parameterName=Intensity})
 ]
}
