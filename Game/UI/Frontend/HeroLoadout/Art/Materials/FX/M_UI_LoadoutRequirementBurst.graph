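# Translucent UI burst effect material, named for hero loadout requirements, built procedurally from radial gradients without textures.
# A central glow combines with a thin horizontal flare made by squashing gradient UVs vertically, then passes through a Grow-controlled SmoothStep.
# A ring, cut from two offset-radius gradients, is added on top.
# The mask, scaled by Alpha, blends between two color parameters, with scalar parameters controlling each shape.
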Material /Game/UI/Frontend/HeroLoadout/Art/Materials/FX/M_UI_LoadoutRequirementBurst
graph {
 properties {
  bUsedWithStaticMesh=false
  blendMode=BLEND_Translucent
  materialDomain=MD_UI
  pixelDepthOffsetMode=PDOM_Legacy
  usedShadingModels=Default Lit
 }
 nodes [
  n1 = MaterialFunctionCall(
   UVs=TextureCoordinate()
   Radius=Multiply(A=Constant() {r=0.5}, B=ScalarParameter() {defaultValue=1, parameterName=Radius})
   Density=ScalarParameter() {parameterName=RadialDensity}
  ) {materialFunction=&RadialGradientExponential}
  n2 = OneMinus(None=@n1.RadialGradientExponential)
  n3 = MaterialFunctionCall(
   UVs=TextureCoordinate() {vTiling=32}
   CenterPosition=Constant2Vector() {g=16, r=0.5}
   Radius=Multiply(A=Constant() {r=0.45}, B=ScalarParameter() {parameterName=Flare})
   Density=Multiply(A=Constant() {r=1.25}, B=ScalarParameter() {parameterName=FlareAlpha})
  ) {materialFunction=&RadialGradientExponential}
  n4 = MaterialFunctionCall(
   Alpha=Add(A=@n1.RadialGradientExponential, B=DotProduct(A=@n2, B=@n3.RadialGradientExponential))
   Min=OneMinus(None=ScalarParameter() {defaultValue=1, parameterName=Grow})
   Max=Constant() {r=1}
  ) {materialFunction=&SmoothStep}
  n5 = TextureCoordinate()
  n6 = Multiply(A=ScalarParameter() {defaultValue=1, parameterName=RingSize}) {constB=0.175}
  n7 = Constant() {r=24}
  n8 = OneMinus(None=ScalarParameter() {parameterName=RingThicc})
  n9 = MaterialFunctionCall(UVs=@n5, Radius=Add(A=@n6, B=Multiply(B=@n8) {constA=-0.025}) {constB=-0.02}, Density=@n7) {materialFunction=&RadialGradientExponential}
  n10 = Multiply(
   A=DotProduct(
    A=MaterialFunctionCall(UVs=@n5, Radius=@n6, Density=@n7) {materialFunction=&RadialGradientExponential}.RadialGradientExponential
    B=OneMinus(None=@n9.RadialGradientExponential)
   )
   B=ScalarParameter() {parameterName=RingAlpha}
  )
  LinearInterpolate(
   A=VectorParameter() {parameterName=Color_1, defaultValue=<rgba 0 0.15 1 0>}.RGB
   B=VectorParameter() {parameterName=Color_2, defaultValue=<rgba 0 1 1 0>}.RGB
   Alpha=Multiply(A=Add(A=@n4.Result, B=@n10), B=ScalarParameter() {defaultValue=1, parameterName=Alpha})
  )
 ]
}
