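# Translucent UI material for a "transient woosh" glow effect, filed with the hero loadout frontend FX.
# Two horizontally offset exponential radial gradients are combined into a centered falloff, which weights a fourth-power vertical UV ramp that strengthens toward the bottom.
# The boosted mask, scaled by an Intensity parameter, blends between two color parameters that default to blue and cyan.
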Material /Game/UI/Frontend/HeroLoadout/Art/Materials/FX/M_UI_TransientWoosh
graph {
 properties {
  bUsedWithStaticMesh=false
  blendMode=BLEND_Translucent
  materialDomain=MD_UI
  pixelDepthOffsetMode=PDOM_Legacy
  usedShadingModels=Default Lit
 }
 nodes [
  n1 = TextureCoordinate()
  n2 = ComponentMask(None=@n1) {g=true}
  n3 = Multiply(A=MaterialFunctionCall(UVs=@n1, CenterPosition=Constant2Vector() {g=0.5, r=0.65}, Radius=Constant() {r=0.5}) {materialFunction=&RadialGradientExponential}.RadialGradientExponential)
  n4 = Multiply(A=MaterialFunctionCall(UVs=TextureCoordinate(), CenterPosition=Constant2Vector() {g=0.5, r=0.35}, Radius=Constant() {r=0.5}) {materialFunction=&RadialGradientExponential}.RadialGradientExponential)
  n5 = LinearInterpolate(B=Power(Base=Multiply(A=@n2)) {constExponent=4}, Alpha=Multiply(A=LinearInterpolate(B=@n3, Alpha=@n4)) {constB=4})
  LinearInterpolate(
   A=VectorParameter() {defaultValue=<rgba 0 0.05 1 1>}.RGB
   B=VectorParameter() {parameterName=Param_1, defaultValue=<rgba 0 0.995 1 1>}.RGB
   Alpha=Multiply(A=Multiply(A=@n5) {constB=1.9}, B=ScalarParameter() {parameterName=Intensity})
  )
 ]
}
